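# Unlit translucent particle material for the Disco drift trail vehicle cosmetic, enabled for Niagara sprite and mesh particles.
# A glow-dot texture is tinted by particle color and scaled by a Bright scalar parameter, then passed through the shared MF_FX_EmissiveAdjustments function.
# The texture times particle alpha is saturated as a separate mask.
# A dynamic parameter, also naming a Bright channel, is present but feeds nothing shown.
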Material /VehicleCosmetics/DriftTrail/DriftTrail_Disco/Materials/M_DiscoBits
graph {
 properties {
  bAutomaticallySetUsageInEditor=false
  bUsedWithNiagaraMeshParticles=true
  bUsedWithNiagaraSprites=true
  blendMode=BLEND_Translucent
  shadingModel=MSM_Unlit
  usedShadingModels=Unlit
 }
 nodes [
  n1 = TextureSample() {texture=/VehicleCosmetics/DriftTrail/DriftTrail_Disco/Textures/T_GlowDot}
  MaterialExpressionVertexColor_1 = ParticleColor()
  MaterialExpressionConstantClamp_1 = Saturate(None=Multiply(A=@n1.RGBA, B=@MaterialExpressionVertexColor_1.A))
  DynamicParameter() {paramNames [Bright,Param2,Param3,Param4]}
  MaterialFunctionCall(Emissive=Multiply(A=ScalarParameter() {defaultValue=1, parameterName=Bright}, B=Multiply(A=@n1.RGB, B=@MaterialExpressionVertexColor_1.RGB))) {materialFunction=&MF_FX_EmissiveAdjustments}
 ]
}
