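# Translucent unlit disco dancefloor material for a vehicle drift trail cosmetic, usable on Niagara ribbons and mesh particles.
# Panned floor diffuse and RGB-mask textures form sine-pulsed colored lights, tinted by particle color, with dynamic parameter Painteds blending from desaturated to full color.
# Opacity combines particle and texture alpha with UV edge fades, switched by ISRIBBON?.
# A time-driven sine on UV1 builds an animated wave offset vector.
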Material /VehicleCosmetics/DriftTrail/DriftTrail_Disco/Materials/M_DiscoDancefloor
graph {
 properties {
  bAutomaticallySetUsageInEditor=false
  bUsedWithNiagaraMeshParticles=true
  bUsedWithNiagaraRibbons=true
  blendMode=BLEND_Translucent
  shadingModel=MSM_Unlit
  usedShadingModels=Unlit
 }
 nodes [
  n1 = Constant() {r=1}
  n2 = ComponentMask(None=TextureCoordinate() {coordinateIndex=1}) {r=true}
  n3 = ScalarParameter() {defaultValue=0.5, parameterName=CountWaves2}
  n4 = ScalarParameter() {defaultValue=4, parameterName=Speed}
  n5 = OneMinus(None=Time())
  n6 = Sine(None=Add(A=Divide(A=@n2, B=@n3), B=Multiply(A=@n4, B=@n5)))
  n7 = ScalarParameter() {defaultValue=5, parameterName=Strength}
  n8 = Power(Base=ComponentMask(None=TextureCoordinate()) {r=true}, Exp=ScalarParameter() {defaultValue=0.5, parameterName=PowerMiddle2})
  n9 = ScalarParameter() {defaultValue=0.5, parameterName=MultPower2}
  AppendVector(A=AppendVector(A=@n1, B=@n1), B=Multiply(A=Multiply(A=@n6, B=@n7), B=Multiply(A=@n8, B=@n9)))
  n10 = ScalarParameter() {defaultValue=5, parameterName=MaxBright}
  n11 = ScalarParameter() {defaultValue=0.15, parameterName=MinBright}
  n12 = Sine(None=Time()) {period=0.45}
  n13 = Constant3Vector() {constant=<rgba 1 0.5 0.1 0>}
  n14 = Panner(
   Coordinate=TextureCoordinate() {uTiling=6.9, vTiling=0.6}
   Speed=MaterialFunctionCall(X=ScalarParameter() {defaultValue=-2.5, parameterName=PannerSpeed}, Y=Constant()) {materialFunction=&MakeFloat2}.Result
  ) {speedX=-2.5}
  n15 = TextureSample(UVs=@n14) {texture=/DealershipSharedLibrary/Assets/Sets/Disco/T_Disco_Floor_Diffuse}
  n16 = TextureSample(UVs=@n14) {texture=/DealershipSharedLibrary/Assets/Sets/Disco/T_Disco_Floor_RGB}
  n17 = ScalarParameter() {defaultValue=5, parameterName=MaxBright}
  n18 = ScalarParameter() {defaultValue=0.15, parameterName=MinBright}
  n19 = Time()
  n20 = Multiply(
   A=Multiply(A=Constant3Vector() {constant=<rgba 0.1 1 0.1 0>}, B=@n16.G)
   B=LinearInterpolate(A=@n17, B=@n18, Alpha=Sine(None=@n19) {period=0.75})
  )
  n21 = Time()
  n22 = Multiply(
   A=Multiply(A=Constant3Vector() {constant=<rgba 0.1 0.1 1 0>}, B=@n16.B)
   B=LinearInterpolate(A=@n17, B=@n18, Alpha=Sine(None=@n21) {period=1.25})
  )
  n23 = Multiply(A=Constant3Vector() {constant=<rgba 1 0.1 0.1 0>}, B=@n16.R)
  n24 = LinearInterpolate(A=@n17, B=@n18, Alpha=Sine(None=Time()) {period=1.35})
  n25 = Add(
   A=Multiply(A=LinearInterpolate(A=@n10, B=@n11, Alpha=@n12), B=Multiply(A=@n13, B=@n15.RGB))
   B=Add(A=Add(A=@n20, B=@n22), B=Multiply(A=@n23, B=@n24))
  )
  n26 = Desaturation(None=@n25)
  MaterialExpressionVertexColor_2 = ParticleColor()
  MaterialFunctionCall(Emissive=LinearInterpolate(
    A=Multiply(A=@MaterialExpressionVertexColor_2.RGB, B=@n26)
    B=Multiply(A=@n25, B=@MaterialExpressionVertexColor_2.RGB)
    Alpha=DynamicParameter() {paramNames [Painteds,Param2,Param3,Param4]}.Painteds
   )) {materialFunction=&MF_FX_EmissiveAdjustments}
  n27 = TextureCoordinate()
  n28 = ScalarParameter() {defaultValue=1, parameterName=UTiling}
  n29 = ScalarParameter() {defaultValue=1, parameterName=VTiling}
  n30 = ComponentMask(None=Multiply(A=@n27, B=MaterialFunctionCall(X=@n28, Y=@n29) {materialFunction=&MakeFloat2}.Result)) {g=true}
  n31 = OneMinus(None=@n30)
  n32 = ScalarParameter() {defaultValue=0.45, parameterName=FrontFade}
  n33 = ScalarParameter() {defaultValue=0.45, parameterName=BackFade}
  n34 = Multiply(A=Multiply(A=Power(Base=@n31, Exp=@n32) {constExponent=1}, B=Power(Base=@n30, Exp=@n33) {constExponent=1})) {constB=5}
  n35 = Power(Base=Sine(None=ComponentMask(None=@n27) {g=true}) {period=1.94}) {constExponent=8}
  n36 = ScalarParameter() {parameterName=ISRIBBON?}
  n37 = Multiply(A=@MaterialExpressionVertexColor_2.A, B=@n15.A)
  n38 = ComponentMask(None=TextureCoordinate()) {r=true}
  n39 = Constant() {r=0.5}
  n40 = OneMinus(None=ComponentMask(None=TextureCoordinate()) {r=true})
  n41 = Constant() {r=0.8}
  n42 = Multiply(A=Multiply(A=Power(Base=@n38, Exp=@n39), B=Power(Base=@n40, Exp=@n41)))
  Saturate(None=Multiply(A=LinearInterpolate(A=@n34, B=@n35, Alpha=@n36), B=Multiply(A=@n37, B=@n42)))
 ]
}
