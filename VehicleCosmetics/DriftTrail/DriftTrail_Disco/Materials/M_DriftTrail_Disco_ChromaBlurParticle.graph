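# Translucent unlit Niagara sprite and ribbon material for the blurred chroma particles of the Disco drift trail cosmetic.
# Emissive multiplies a panning rainbow gradient and particle color by noise from the Boost Helix ChromaBlur function.
# Dynamic parameters drive blur, vertical pan and threshold, while Painted times particle alpha blends from desaturated, re-tinted color to rainbow.
# Opacity combines a soft radial UV mask, depth fade and particle alpha.
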Material /VehicleCosmetics/DriftTrail/DriftTrail_Disco/Materials/M_DriftTrail_Disco_ChromaBlurParticle
graph {
 properties {
  bAutomaticallySetUsageInEditor=false
  bUsedWithNiagaraRibbons=true
  bUsedWithNiagaraSprites=true
  blendMode=BLEND_Translucent
  shadingModel=MSM_Unlit
  usedShadingModels=Unlit
 }
 nodes [
  MaterialExpressionVertexColor_0 = ParticleColor()
  n1 = TextureSample(UVs=Panner(Coordinate=TextureCoordinate()) {speedX=-1}) {texture=/DealershipSharedLibrary/Assets/Sets/Disco/T_Rainbow_Gradient_Diffuse}
  n2 = DynamicParameter() {paramNames [Painted, Scale, Speed_Y, OpacityThreshold]}
  n3 = MaterialFunctionCall!(
   ExtraNoiseTile=ScalarParameter() {defaultValue=1, parameterName=ExtraNoiseTile}
   NoiseTile_Y=ScalarParameter() {parameterName=NoiseTile_Y}
   PanSpeed_Y=ComponentMask(None=@n2.Speed_Y) {r=true}
   PanSpeed_X=Constant()
   ExtraNoiseSpeedMult=ScalarParameter() {defaultValue=1, parameterName=ExtraNoiseSpeedMult}
   NoiseTile_X=ScalarParameter() {defaultValue=1, parameterName=NoiseTile_X}
   ExtraNoise=StaticBoolParameter() {parameterName=ExtraNoise}
   BlurFalloff=ComponentMask(None=@n2.Scale) {r=true}
   AddNoise=StaticBoolParameter() {defaultValue=true, parameterName=Noise}
   NoiseTexture=TextureObjectParameter() {parameterName=NoiseTexture, texture=/DealershipSharedLibrary/FX/Textures/Noise/T_VCVFX_Caustics_A}
   Extra_NoiseTexture=TextureObjectParameter() {parameterName=ExtraNoiseTexture, texture=&T_VCVFX_Caustics_A}
  ) {materialFunction=/VehicleCosmetics/Boosts/Boost_Helix/Materials/MF_Boost_ChromaBlur}
  n4 = Add(A=Multiply(A=Multiply(A=@n1.RGB, B=@MaterialExpressionVertexColor_0.RGB), B=Power(Base=@n3.Opacity, Exp=@n2.OpacityThreshold))) {constB=0.1}
  n5 = Desaturation(None=@n4)
  MaterialFunctionCall(Emissive=LinearInterpolate(A=Multiply(A=@n5, B=@MaterialExpressionVertexColor_0.RGB), B=@n4, Alpha=Multiply(A=@MaterialExpressionVertexColor_0.A, B=@n2.Painted))) {materialFunction=&MF_FX_EmissiveAdjustments}
  n6 = Add(A=TextureCoordinate(), B=Constant() {r=-0.5})
  n7 = ComponentMask(None=@n6) {g=true}
  n8 = ComponentMask(None=@n6) {r=true}
  n9 = Add(A=Multiply(A=@n7, B=@n7), B=Multiply(A=@n8, B=@n8))
  n10 = ScalarParameter() {parameterName=CenterMaskRadius}
  MaterialExpressionConstantClamp_4 = Saturate(None=Multiply(A=Subtract(A=@n9, B=@n10), B=ScalarParameter() {defaultValue=6, parameterName=CenterMaskSharpness}))
  n11 = OneMinus(None=@MaterialExpressionConstantClamp_4)
  Multiply(
   A=@MaterialExpressionVertexColor_0.A
   B=MaterialFunctionCall(Opacity In=Power(Base=@n11) {constExponent=3}, DepthFade Distance=Constant() {r=16}) {materialFunction=&MF_VCVFX_DepthFade}.Result
  )
 ]
}
