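# Translucent unlit Niagara sprite material for a soft glowing dot in the Disco drift trail vehicle cosmetic.
# A rotating rainbow gradient is multiplied by squared particle colour and a soft dot mask, which also combines with particle alpha.
# The Painted dynamic parameter blends from a desaturated, particle-tinted version to the rainbow one.
# The blend feeds MF_FX_EmissiveAdjustments as emissive input.
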Material /VehicleCosmetics/DriftTrail/DriftTrail_Disco/Materials/M_Soft_Dot
graph {
 properties {
  bAutomaticallySetUsageInEditor=false
  bUsedWithNiagaraSprites=true
  blendMode=BLEND_Translucent
  shadingModel=MSM_Unlit
  usedShadingModels=Unlit
 }
 nodes [
  MaterialExpressionVertexColor_0 = ParticleColor()
  n1 = TextureSample() {texture=/VehicleCosmetics/Boosts/Boost_Drofb/Textures/T_VCVFX_Boost_Drofb_Dot_Soft_Mask}
  MaterialExpressionConstantClamp_0 = Saturate(None=Multiply(A=@MaterialExpressionVertexColor_0.A, B=@n1.RGB))
  n2 = Rotator(Coordinate=TextureCoordinate()) {speed=2}
  MaterialExpressionMeshEmitterVertexColor_0 = ParticleColor()
  MaterialExpressionVertexColor_1 = ParticleColor()
  n3 = Multiply(
   A=Multiply(
    A=TextureSample(UVs=@n2) {texture=/DealershipSharedLibrary/Assets/Sets/Disco/T_Rainbow_Gradient_Centre_Diffuse}.RGB
    B=Multiply(A=@MaterialExpressionMeshEmitterVertexColor_0.RGB, B=@MaterialExpressionVertexColor_1.RGB)
   )
   B=@n1.RGB
  )
  n4 = Desaturation(None=@n3)
  MaterialFunctionCall(Emissive=LinearInterpolate(
    A=Multiply(A=@n4, B=@MaterialExpressionVertexColor_0.RGB)
    B=@n3
    Alpha=DynamicParameter() {parameterIndex=1, paramNames [Painted,Param2,Param3,Param4]}.Painted
   )) {materialFunction=&MF_FX_EmissiveAdjustments}
 ]
}
