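# Niagara system, named as a backup, for the Disco drift trail vehicle cosmetic.
# Paired emitters layer ribbons, multicolor sparkle sprites, looping glow sprites and wheel-wrap meshes, with one of each pair offset 300 units in local Y.
# Particles track wheel sockets on the vehicle's skeletal mesh, partitioned per wheel and masked by FN_DisableTireFX.
# User.bIsPainted switches many colors and material parameters between defaults and painted values, alongside per-paint override tables.
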NiagaraSystem /VehicleCosmetics/DriftTrail/DriftTrail_Disco/NS_DriftTrail_Disco_Backup
dependencies {
 usedDataInterfaces [
  &NiagaraDataInterfaceCurve
  &NiagaraDataInterfaceArrayColor
  &NiagaraDataInterfaceColorCurve
  &NiagaraDataInterfaceArrayFloat3
  &NiagaraDataInterfaceSkeletalMesh
 ]
 usedDynamicInputs []
 usedModules [
  &self:SystemSpawnScript.SystemScriptSource.SystemScriptGraph.NiagaraNodeAssignment_0.SetVariables_#1
  &FN_UnpackVehicleFlags
  /VehicleCosmetics/FX/Niagara/Modules/VC_SetVehicleSpeedAttribs
  &SystemState
  /Game/Effects/Niagara/Modules/Vehicle/FN_VehicleFlags_Cheatcodes
 ]
}
emitters {
 NE_Disco_Ribbon {
  emitterUpdateScript [
   &EmitterState()
   &VC_RateQualityHandler(Epic Spawn Rate=30, High Spawn Rate=30, Low Spawn Rate=25, Medium Spawn Rate=25, Use Custom Rate?=true)
   &VC_SpawnRibbon_v1(Number of Ribbons=2, Spawn Time=-1, SpawnRate=$Emitter.Spawn Rate)
  ]
  particleSpawnScript [
   &VC_PaintedVariantsOverride() {enabled=false}
   &VC_InitializeRibbon_v1(
    Color=&MakeCustomLinearColorFromBool(Bool=$User.bIsPainted, True Color=$User.PaintedColor, False Color=<rgba 1 1 1 1>)
    Custom Offset=true
    Enable Skeletal Position=true
    Facing Axis:ENiagaraOrientationAxis=::ZAxis
    Lifetime=0.55
    Local Vector Offset=&AddVectorToPosition(Position=<0 0 -30>, Vector=Linked() {linkedVariable {name=User.offset, type=NiagaraPosition}})
    Offset Amount=1
    Ribbon Width=40
   )
   &AddVelocity(Velocity=<1000 0 0>, Velocity Speed Scale=-2) {module=AddVelocity001}
  ]
  particleUpdateScript [
   &FN_DisableTireFX(Mask Index Attribute:NiagaraInt32=$Particles.SpawnGroup, Mask Index FL=2, Mask Index FR=3, Mask Index RL=0, Mask Index RR=1)
   &ScaleColor(
    Scale Alpha=&MakeCustomFloatFromBool(Bool=$User.bIsPainted, True Float=1.5, False Float=1)
    Scale RGB=&MakeCustomVectorFromBool(
     Bool=$User.bIsPainted
     True Vector=&MakeVectorFromLinearColorRGB(Color=$User.PaintedColor)
     False Vector=<1 1 1>
    )
    ScaleA=true
    ScaleRGB=true
   )
   &ParticleState()
   &InheritSourceMovement(
    Applied Position Scale=&VectorFromFloat(Value=&FloatFromCurve(FloatCurve=DataInterface() {propertyValues {Curve {keys [
          {interpMode=RCIM_Cubic, tangentMode=RCTM_Break, time=0, value=1}
          {arriveTangent=-2.9025, interpMode=RCIM_Cubic, tangentMode=RCTM_Break, time=1, value=0}
         ]}}}))
    ApplyForce=false
    ApplyPosition=true
    ApplyPositionFalloff=false
   )
   &SolveForcesAndVelocity(Clamp Velocity=false, Limit Acceleration=false)
   &DynamicMaterialParameters(
    Index 0 Param 1=&MakeCustomFloatFromBool(Bool=$User.bIsPainted, True Float=0, False Float=1)
    Module.Index0Float0Write=true
    Module.Index0Float1Write=true
    Module.Index0Float2Write=true
    Module.Index0Float3Write=true
   )
   &VC_UpdateRibbon_v1(Enable Skeletal Position=true)
  ]
  rendererClasses [&NiagaraRibbonRendererProperties]
  renderers [{rendererClass=&NiagaraRibbonRendererProperties}]
 }
 NE_Disco_Ribbon001 {
  emitterUpdateScript [
   &EmitterState()
   &VC_RateQualityHandler(Epic Spawn Rate=30, High Spawn Rate=30, Low Spawn Rate=25, Medium Spawn Rate=25, Use Custom Rate?=true)
   &VC_SpawnRibbon_v1(Number of Ribbons=2, Spawn Time=-1, SpawnRate=$Emitter.Spawn Rate)
  ]
  particleSpawnScript [
   &VC_PaintedVariantsOverride() {enabled=false}
   &VC_InitializeRibbon_v1(
    Color=&MakeCustomLinearColorFromBool(Bool=$User.bIsPainted, True Color=$User.PaintedColor, False Color=<rgba 1 1 1 1>)
    Custom Offset=true
    Enable Skeletal Position=true
    Facing Axis:ENiagaraOrientationAxis=::ZAxis
    Lifetime=0.55
    Local Vector Offset=&AddVectorToPosition(Position=<0 300 -30>, Vector=Linked() {linkedVariable {name=User.offset, type=NiagaraPosition}})
    Offset Amount=1
    Ribbon Width=40
   )
   &AddVelocity(Velocity=<1000 0 0>, Velocity Speed Scale=-2) {module=AddVelocity001}
  ]
  particleUpdateScript [
   &FN_DisableTireFX(Mask Index Attribute:NiagaraInt32=$Particles.SpawnGroup, Mask Index FL=2, Mask Index FR=3, Mask Index RL=0, Mask Index RR=1)
   &ScaleColor(
    Scale Alpha=&MakeCustomFloatFromBool(Bool=$User.bIsPainted, True Float=1.5, False Float=1)
    Scale RGB=&MakeCustomVectorFromBool(
     Bool=$User.bIsPainted
     True Vector=&MakeVectorFromLinearColorRGB(Color=$User.PaintedColor)
     False Vector=<1 1 1>
    )
    ScaleA=true
    ScaleRGB=true
   )
   &ParticleState()
   &InheritSourceMovement(
    Applied Position Scale=&VectorFromFloat(Value=&FloatFromCurve(FloatCurve=DataInterface() {propertyValues {Curve {keys [
          {interpMode=RCIM_Cubic, tangentMode=RCTM_Break, time=0, value=1}
          {arriveTangent=-2.9025, interpMode=RCIM_Cubic, tangentMode=RCTM_Break, time=1, value=0}
         ]}}}))
    ApplyForce=false
    ApplyPosition=true
    ApplyPositionFalloff=false
   )
   &SolveForcesAndVelocity(Clamp Velocity=false, Limit Acceleration=false)
   &DynamicMaterialParameters(
    Index 0 Param 1=&MakeCustomFloatFromBool(Bool=$User.bIsPainted, True Float=0, False Float=1)
    Module.Index0Float0Write=true
    Module.Index0Float1Write=true
    Module.Index0Float2Write=true
    Module.Index0Float3Write=true
   )
   &VC_UpdateRibbon_v1(Enable Skeletal Position=true)
  ]
  rendererClasses [&NiagaraRibbonRendererProperties]
  renderers [{rendererClass=&NiagaraRibbonRendererProperties}]
 }
 NE_VC_DriftTrail_Sprite_Parent {
  emitterUpdateScript [
   &EmitterState()
   &SpawnRate(SpawnRate=40, Use Spawn Probability=false)
   &SpawnRate(SpawnGroup=1, SpawnRate=15, Use Spawn Probability=false) {module=SpawnRate001}
  ]
  particleSpawnScript [
   &VC_PaintedVariantsOverride() {enabled=false}
   &InitializeParticle(
    Color=&MakeCustomLinearColorFromBool(
     Bool=$User.bIsPainted
     True Color=$User.LParticleColor
     False Color=&SelectLinearColorFromArray(
      Array Sampling Mode=::NewEnumerator0
      Direct Array Index=?
      Color Selection Array=DataInterface() {propertyValues {ColorData [
         <rgba 1 0.0938 0.0938 0.5>
         <rgba 1 1 0.1042 0.5>
         <rgba 0.1458 1 0.1458 0.5>
         <rgba 0.1146 1 1 0.5>
         <rgba 0.1146 0.1146 1 0.5>
         <rgba 0.8637 0.1354 1 0.5>
        ]}}
      Random Seed=0
     )
    )
    Sprite Size Mode=::NewEnumerator3
    Uniform Sprite Size=/VehicleCosmetics/FX/Niagara/Modules/MergeEmitters/VC_FloatBySpawnGroup(
     SpawnGroup or ExecIndex:NiagaraInt32=0
     CustomInt:NiagaraInt32=0
     MaximumSpawnGroups:NiagaraInt32=2
     Float SG 0:NiagaraFloat=&RandomRangeFloat(Minimum=6, Maximum=18)
     Float SG 1:NiagaraFloat=&RandomRangeFloat(Minimum=18, Maximum=30)
     Float SG 2:NiagaraFloat=0
     Float SG 3:NiagaraFloat=0
     Float SG 4:NiagaraFloat=0
     Float SG 5:NiagaraFloat=0
    )
    UsePositionOffset=true
    Write Material Random=true
   )
   &PartitionParticles()
   &SkeletalMeshLocation(
    Bone / Socket Index=&Add_Integer(A=$Particles.PartitionParticles.Partition, B=4)
    Bone / Socket Sampling Mode=::NewEnumerator5
    Interpolate Mesh Sampling=false
    MaskSpawnGroup=false
    Mesh Sampling Type=::NewEnumerator6
    Sampled Position Offset=&AddVectorToPosition(Position=<-60 0 -20>, Vector=Linked() {linkedVariable {name=User.offset, type=NiagaraPosition}})
    Skeletal Mesh=$System.SkeletalMesh
   )
   &AddVelocity(
    Cone Angle=10
    Module.UseConeFalloff=false
    Module.UseVelDistribution=false
    Override Seed=false
    Velocity Mode=::NewEnumerator2
    Velocity Speed=-400
   )
   /SetVariables(Particles.SVisTag:NiagaraInt32=&SpawnGroup()) {isSetParameters=true, module=SetVariables_#2}
  ]
  particleUpdateScript [
   &FN_DisableTireFX(
    Mask Index Attribute:NiagaraInt32=$Particles.PartitionParticles.Partition
    Mask Index FL=2
    Mask Index FR=3
    Mask Index RL=0
    Mask Index RR=1
   )
   &Color(Color=&VC_DynamicScaleColor(
     Color to Scale:LinearColor=&MakeCustomLinearColorFromBool(
      Bool=$User.bIsPainted
      True Color=$User.PaintedColor
      False Color=&SelectLinearColorFromArray(
       Array Sampling Mode=::NewEnumerator0
       Direct Array Index=?
       Color Selection Array=DataInterface() {propertyValues {ColorData [<rgba 1 0 0.3152 1>, <rgba 1 1 1 1>, <rgba 0.5674 1 0.0923 1>, <rgba 0 0 0 0>, <rgba 0 0 0 0>, <rgba 0 0 0 0>]}}
       Random Seed=0
      )
     )
     RGB Scale:Vector3f=<1 1 1>
     Alpha Scale:NiagaraFloat=1
    )) {enabled=false}
   &ScaleColor(
    Scale Alpha=&FloatFromCurve(FloatCurve=DataInterface() {propertyValues {Curve {keys [<at 0 2>,<at 1 0>]}}}, Scale Curve=0.5)
    Scale RGB=<2 2 2>
    ScaleA=true
    ScaleRGB=true
   )
   &Drag(
    Drag=&FloatFromCurve(FloatCurve=DataInterface() {propertyValues {Curve {keys [{tangentMode=RCTM_User, time=0, value=0}, {tangentMode=RCTM_User, time=1, value=0.5}]}}})
    Use Linear Drag=true
    Use Rotational Drag=true
   )
   &ParticleState()
   &ScaleSpriteSize(Uniform Curve Sprite Scale=DataInterface() {propertyValues {Curve {keys [{tangentMode=RCTM_User, time=0, value=0.5}, {tangentMode=RCTM_User, time=1, value=1}]}}})
   /SetVariables() {isSetParameters=true, module=SetVariables_#3}
   /SetVariables(Particles.StarBitColor:LinearColor=&VC_DynamicScaleColor(
     Color to Scale:LinearColor=&MakeCustomLinearColorFromBool(Bool=$User.bIsPainted, True Color=$User.StarCrossColor, False Color=<rgba 5 5 5 1>)
     RGB Scale:Vector3f=<5 5 5>
     Alpha Scale:NiagaraFloat=&FloatFromCurve(FloatCurve=DataInterface() {propertyValues {Curve {keys [<at 0 2>,<at 1 0>]}}})
    )) {isSetParameters=true, module=SetVariables_#4}
   &SolveForcesAndVelocity(Clamp Velocity=false, Limit Acceleration=false)
  ]
  rendererClasses [&NiagaraSpriteRendererProperties]
  renderers [
   {rendererClass=&NiagaraSpriteRendererProperties}
   {rendererClass=&NiagaraSpriteRendererProperties, rendererIndex=1}
  ]
 }
 NE_VC_DriftTrail_Sprite_Parent001 {
  emitterSpawnScript [
   /SetVariables(
    Emitter.Desaturation:NiagaraFloat=&SelectFloatFromArray(Direct Array Index=$System.ColorVariantID, Float Selection Array=DataInterface() {propertyValues {FloatData [0,1,1,1,1,1,1,1,1,1,1,1,1,1]}})
    Emitter.EmissiveGain:NiagaraFloat=&SelectFloatFromArray(
     Direct Array Index=$System.ColorVariantID
     Float Selection Array=DataInterface() {propertyValues {FloatData [1.5,1.5,1.5,2,15,10,10,5,5,1.5,6,6,5,35]}}
    )
    Emitter.HueShift:NiagaraFloat=&SelectFloatFromArray(
     Direct Array Index=$System.ColorVariantID
     Float Selection Array=DataInterface() {propertyValues {FloatData [0.5,0,0,0,0,0,0,0,0,0,0,0,0,0]}}
    )
   ) {isSetParameters=true}
   /SetVariables(Emitter.No Painted:NiagaraBool=&Boolean_NotOperation(Bool=$User.bIsPainted)) {isSetParameters=true}
   /SetVariables(Emitter.Painted:NiagaraBool=&Boolean_NotOperation(Bool=$StackContext.No Painted)) {isSetParameters=true}
  ]
  emitterUpdateScript [
   &EmitterState(
    Inactive Response=::NewEnumerator1
    Life Cycle Mode=::NewEnumerator1
    Loop Behavior=::NewEnumerator1
    Loop Duration Mode=::NewEnumerator1
    UseLoopDelay=false
   )
   &SpawnBurst_Instantaneous(Spawn Count=2, Use Spawn Probability=false, UseLoopCountLimit=false)
   &SpawnBurst_Instantaneous(Spawn Count=2, Spawn Group=1, Use Spawn Probability=false, UseLoopCountLimit=false) {module=SpawnBurst_Instantaneous001}
   &SpawnBurst_Instantaneous(Spawn Count=2, Spawn Group=2, Use Spawn Probability=false, UseLoopCountLimit=false) {module=SpawnBurst_Instantaneous002}
   &SpawnBurst_Instantaneous(Spawn Count=2, Spawn Group=3, Use Spawn Probability=false, UseLoopCountLimit=false) {module=SpawnBurst_Instantaneous003}
   &SpawnRate(Use Spawn Probability=false) {enabled=false}
  ]
  particleSpawnScript [
   &VC_PaintedVariantsOverride(
    BurntSienna=<rgba 0.0156 0.0048 0.0019 1>
    R1 Black:LinearColor=<rgba 0.2552 0.2552 0.2552 1>
    R1 BurntSienna:LinearColor=<rgba 0.8698 0.2783 0 1>
    R1 Cobalt:LinearColor=<rgba 0.0894 0.0977 0.5365 1>
    R1 Crimson:LinearColor=<rgba 0.8438 0.1494 0.1682 1>
    R1 DefaultColor=<rgba 4 4 4 0.8348>
    R1 ForestGreen:LinearColor=<rgba 0.1135 0.7031 0.1433 1>
    R1 Grey:LinearColor=<rgba 0.3854 0.3854 0.3854 1>
    R1 Lime:LinearColor=<rgba 0.5092 0.7917 0.2268 1>
    R1 Orange:LinearColor=<rgba 3 1.3313 0.4375 1>
    R1 Pink:LinearColor=<rgba 0.8229 0.3086 0.6359 1>
    R1 Purple:LinearColor=<rgba 0.4458 0.2831 0.6795 1>
    R1 Saffron:LinearColor=<rgba 0.9 0.7512 0.2625 1>
    R1 SkyBlue:LinearColor=<rgba 0 0.4419 0.6641 1>
    R1 TitaniumWhite:LinearColor=<rgba 1 1 1 1>
    R2 Black:LinearColor=<rgba 0.0104 0.0104 0.0104 0.8>
    R2 BurntSienna:LinearColor=<rgba 0.0312 0.0096 0.0039 0.313>
    R2 Cobalt:LinearColor=<rgba 0.0934 0 2.5 0.3391>
    R2 Crimson:LinearColor=<rgba 1.3125 0.082 0.1154 0.5391>
    R2 DefaultColor:LinearColor=&SelectLinearColorFromArray(
     Array Sampling Mode=::NewEnumerator2
     Direct Array Index=&MakeIntFromFloat(Float=&Sine(Normalized Angle=$Emitter.Age, Period=0.5, Scale=1, Bias=0), Conversion Mode=::NewEnumerator1)
     Direct Array Mode=::NewEnumerator1
     Interpolation=&Sine(Normalized Angle=$Emitter.Age, Period=0.5, Scale=1, Bias=0)
     Color Selection Array=DataInterface() {propertyValues {ColorData [<rgba 1 0.5219 0.0036 0.2>, <rgba 1 0.0965 0.1729 0.8174>]}}
     Random Seed=?
    )
    R2 Grey:LinearColor=<rgba 0.2031 0.2031 0.2031 0.1739>
    R2 Lime:LinearColor=<rgba 0.46 0.8073 0 0.5>
    R2 Orange:LinearColor=<rgba 0.8 0.0852 0 0.5>
    R2 Pink:LinearColor=<rgba 0.1946 0 0.2 8>
    R2 Purple:LinearColor=<rgba 0.0799 0.0125 0.4375 2>
    R2 Saffron:LinearColor=<rgba 0.7708 0.5465 0.0819 0.7217>
    R2 SkyBlue:LinearColor=<rgba 0 0.9981 1.5 0.7043>
    R2 TitaniumWhite:LinearColor=<rgba 1.5 1.5 1.5 1>
    R3 Black:LinearColor=<rgba 0.0087 0.0087 0.0087 4>
    R3 BurntSienna:LinearColor=<rgba 0.026 0.008 0.0032 2>
    R3 Cobalt:LinearColor=<rgba 0.0813 0.1482 0.6 0.8957>
    R3 Crimson:LinearColor=<rgba 0.724 0 0.0196 0.8>
    R3 DefaultColor:LinearColor=<rgba 0.0677 0.0557 0.0433 2>
    R3 ForestGreen:LinearColor=<rgba 0.0673 0.3229 0.0802 0.7652>
    R3 Grey:LinearColor=<rgba 0.0208 0.0208 0.0208 3.2609>
    R3 Lime:LinearColor=<rgba 0.8021 1 0.6042 0.2435>
    R3 Orange:LinearColor=<rgba 0.5365 0.1754 0.0838 0.4043>
    R3 Pink:LinearColor=<rgba 0.7135 0.1412 0.5054 0.4696>
    R3 Purple:LinearColor=<rgba 0.0597 0.1043 1 0.5>
    R3 Saffron:LinearColor=<rgba 0.9 0.6964 0.2557 0.5478>
    R3 SkyBlue:LinearColor=<rgba 0 0.6654 1 0.2609>
    R3 TitaniumWhite:LinearColor=<rgba 0.9115 0.9115 0.9115 0.2522>
    R4 Black:LinearColor=<rgba 0.1615 0.1615 0.1615 2>
    R4 BurntSienna:LinearColor=<rgba 0.3125 0.102 0.0456 5>
    R4 Cobalt:LinearColor=<rgba 0 0.1092 0.4896 5>
    R4 Crimson:LinearColor=<rgba 1.5 0.1172 0.1547 5>
    R4 DefaultColor:LinearColor=<rgba 1.5 1.5 1.5 2>
    R4 ForestGreen:LinearColor=<rgba 0.4219 3 0.5521 2>
    R4 Grey:LinearColor=<rgba 0.0781 0.0781 0.0781 2>
    R4 Lime:LinearColor=<rgba 0.8711 1.5 0.2422 2.5>
    R4 Orange:LinearColor=<rgba 1.5 0.4205 0.1317 5>
    R4 Pink:LinearColor=<rgba 1.5 0.4949 1.1269 5>
    R4 Purple:LinearColor=<rgba 0.5163 0 3 5>
    R4 Saffron:LinearColor=<rgba 1.5 1.0424 0 2>
    R4 SkyBlue:LinearColor=<rgba 0 3.3269 5 1>
    R4 TitaniumWhite:LinearColor=<rgba 5 5 5 2>
    Use R1 Painted Overrides=true
    Use R2 Painted Overrides=true
    Use R3 Painted Overrides=true
    Use R4 Painted Overrides=true
   )
   /SetVariables(Particles.DVisTag:NiagaraInt32=&SpawnGroup()) {isSetParameters=true, module=SetVariables_#5}
   &InitializeParticle(
    Color Mode=::NewEnumerator0
    Lifetime=$System.CurrentLoopDuration
    Sprite Rotation Mode=::NewEnumerator2
    Sprite Size Mode=::NewEnumerator3
    Uniform Sprite Size=&VC_FloatBySpawnGroup(
     SpawnGroup or ExecIndex:NiagaraInt32=0
     CustomInt:NiagaraInt32=0
     MaximumSpawnGroups:NiagaraInt32=3
     Float SG 0:NiagaraFloat=48
     Float SG 1:NiagaraFloat=6.5
     Float SG 2:NiagaraFloat=58
     Float SG 3:NiagaraFloat=45
     Float SG 4:NiagaraFloat=0
     Float SG 5:NiagaraFloat=0
    )
    UsePositionOffset=false
    Write Material Random=true
   )
   &PartitionParticles()
   &SkeletalMeshLocation(
    Bone / Socket Index=&Add_Integer(A=$Particles.PartitionParticles.Partition, B=4)
    Bone / Socket Sampling Mode=::NewEnumerator5
    Interpolate Mesh Sampling=false
    MaskSpawnGroup=false
    Mesh Sampling Type=::NewEnumerator6
    Sampled Position Offset=&AddVectorToPosition(Position=<0 0 0>, Vector=Linked() {linkedVariable {name=User.offset, type=NiagaraPosition}})
    Skeletal Mesh=$System.SkeletalMesh
   )
   &AddVelocity(Velocity=&VC_Mask_Vectors(
     FrontEnd Value:Vector3f=<-1500 0 0>
     blsFrontEndPreview:NiagaraBool=false
     SpeedMin Value:Vector3f=<-800 0 0>
     SpeedMax Value:Vector3f=<-1 0 0>
     Speed Alpha:NiagaraFloat=$System.Speed_Alpha
    )) {enabled=false}
  ]
  particleUpdateScript [
   &FN_DisableTireFX(
    Mask Index Attribute:NiagaraInt32=$Particles.PartitionParticles.Partition
    Mask Index FL=2
    Mask Index FR=3
    Mask Index RL=0
    Mask Index RR=1
   )
   &ParticleState(Kill Particles When Lifetime Has Elapsed=false, Loop Particles Lifetime=true)
   &SpriteRotationRate(Rotation Rate=&MaskFloatBySpawnGroup(
     Masked Float=-120
     Passthrough Float=&MaskFloatBySpawnGroup(Masked Float=0, Passthrough Float=0)
     Spawn Group Mask=1
    ))
   &Color(Color=&MakeCustomLinearColorFromBool(Bool=$User.bIsPainted, True Color=$User.RaysColor, False Color=<rgba 0.3281 0.2702 0.21 0.5>)) {enabled=false}
   &Color(Color=&MakeCustomLinearColorFromBool(
     Bool=$User.bIsPainted
     True Color=$User.RaysColor
     False Color=&ColorFromCurve(
      ColorCurve=DataInterface() {propertyValues {
        AlphaCurve {keys [<at 0 1 RCIM_Cubic>,<at 1 1 RCIM_Cubic>]}
        BlueCurve {keys [<at 0 0.0844>,<at 0.1928 0.0908>,<at 0.4876 0.5271>,<at 0.7854 0.8632>,<at 1 0.117>]}
        GreenCurve {keys [<at 0 0.3564>,<at 0.1928 0.107>,<at 0.4876 0.0999>,<at 0.7854 0.2664>,<at 1 0.9734>]}
        RedCurve {keys [<at 0 1>,<at 0.1928 1>,<at 0.4876 0.3763>,<at 0.7854 0.0409>,<at 1 0.1022>]}
       }}
      CurveIndex=&Multiply_Float(A=$Emitter.Age, B=1)
     )
    )) {enabled=false, module=Color001}
   &DynamicMaterialParameters(
    Index 0 Param 1=HlslExpression() {hlslExpression=Engine.Time * -.8}
    Index 0 Param 3=2
    Index 0 Param 4=0.4
    Index 2 Param 1=1.5
    Index 2 Param 2=1.5
    Module.Index0Float0Write=true
    Module.Index0Float1Write=true
    Module.Index0Float2Write=true
    Module.Index0Float3Write=true
    Module.Index1Float0Write=true
    Module.Index1Float1Write=true
    Module.Index1Float2Write=true
    Module.Index1Float3Write=true
    Module.Index2Float0Write=true
    Module.Index2Float1Write=true
    Module.Index2Float2Write=true
    Module.Index2Float3Write=true
    Module.Index3Float0Write=true
    Module.Index3Float1Write=true
    Module.Index3Float2Write=true
    Module.Index3Float3Write=true
    Write Parameter Index 1=::NewEnumerator1
    Write Parameter Index 2=::NewEnumerator1
    Write Parameter Index 3=::NewEnumerator1
   )
   &SkeletalMeshLocation(
    Bone / Socket Index=&Add_Integer(A=$Particles.PartitionParticles.Partition, B=4)
    Bone / Socket Sampling Mode=::NewEnumerator5
    MaskSpawnGroup=false
    Mesh Sampling Type=::NewEnumerator6
    Sampled Position Offset=&AddVectorToPosition(Position=<-65 0 -20>, Vector=Linked() {linkedVariable {name=User.offset, type=NiagaraPosition}})
    Skeletal Mesh=$System.SkeletalMesh
   ) {module=SkeletalMeshLocation001}
   /SetVariables(
    Particles.GlowColor=&VC_DynamicScaleColor(
     Color to Scale:LinearColor=&MakeCustomLinearColorFromBool(Bool=$User.bIsPainted, True Color=$User.GlowColor, False Color=<rgba 1 0.5219 0.0036 0.2>)
     RGB Scale:Vector3f=<1 1 1>
     Alpha Scale:NiagaraFloat=0.5
    )
    Particles.GlowColor2:LinearColor=<rgba 0.7609 0.1023 1 0.2>
    Particles.SoftDotColor:LinearColor=&VC_DynamicScaleColor(
     Color to Scale:LinearColor=&MakeCustomLinearColorFromBool(
      Bool=$User.bIsPainted
      True Color=&Multiply_LinearColorByFloat(Linear Color=$User.SoftDotColor, Float=2)
      False Color=<rgba 4 4 4 1>
     )
     RGB Scale:Vector3f=<1 1 1>
     Alpha Scale:NiagaraFloat=1
    )
   ) {isSetParameters=true}
   &SolveForcesAndVelocity(Clamp Velocity=false, Limit Acceleration=false)
  ]
  rendererClasses [&NiagaraSpriteRendererProperties]
  renderers [
   {rendererClass=&NiagaraSpriteRendererProperties}
   {rendererClass=&NiagaraSpriteRendererProperties, rendererIndex=1}
   {rendererClass=&NiagaraSpriteRendererProperties, rendererIndex=2}
   {rendererClass=&NiagaraSpriteRendererProperties, rendererIndex=3}
   {rendererClass=&NiagaraSpriteRendererProperties, rendererIndex=4}
   {rendererClass=&NiagaraSpriteRendererProperties, rendererIndex=5}
  ]
 }
 NE_VC_DriftTrail_Sprite_Parent003 {
  emitterUpdateScript [
   &EmitterState()
   &SpawnRate(SpawnRate=40, Use Spawn Probability=false)
   &SpawnRate(SpawnGroup=1, SpawnRate=15, Use Spawn Probability=false) {module=SpawnRate001}
  ]
  particleSpawnScript [
   &VC_PaintedVariantsOverride() {enabled=false}
   &InitializeParticle(
    Color=&MakeCustomLinearColorFromBool(
     Bool=$User.bIsPainted
     True Color=$User.LParticleColor
     False Color=&SelectLinearColorFromArray(
      Array Sampling Mode=::NewEnumerator0
      Direct Array Index=?
      Color Selection Array=DataInterface() {propertyValues {ColorData [
         <rgba 1 0.0938 0.0938 0.5>
         <rgba 1 1 0.1042 0.5>
         <rgba 0.1458 1 0.1458 0.5>
         <rgba 0.1146 1 1 0.5>
         <rgba 0.1146 0.1146 1 0.5>
         <rgba 0.8637 0.1354 1 0.5>
        ]}}
      Random Seed=0
     )
    )
    Sprite Size Mode=::NewEnumerator3
    Uniform Sprite Size=&VC_FloatBySpawnGroup(
     SpawnGroup or ExecIndex:NiagaraInt32=0
     CustomInt:NiagaraInt32=0
     MaximumSpawnGroups:NiagaraInt32=2
     Float SG 0:NiagaraFloat=&RandomRangeFloat(Minimum=6, Maximum=18)
     Float SG 1:NiagaraFloat=&RandomRangeFloat(Minimum=18, Maximum=30)
     Float SG 2:NiagaraFloat=0
     Float SG 3:NiagaraFloat=0
     Float SG 4:NiagaraFloat=0
     Float SG 5:NiagaraFloat=0
    )
    UsePositionOffset=true
    Write Material Random=true
   )
   &PartitionParticles()
   &SkeletalMeshLocation(
    Bone / Socket Index=&Add_Integer(A=$Particles.PartitionParticles.Partition, B=4)
    Bone / Socket Sampling Mode=::NewEnumerator5
    Interpolate Mesh Sampling=false
    MaskSpawnGroup=false
    Mesh Sampling Type=::NewEnumerator6
    Sampled Position Offset=&AddVectorToPosition(Position=<-60 300 -20>, Vector=Linked() {linkedVariable {name=User.offset, type=NiagaraPosition}})
    Skeletal Mesh=$System.SkeletalMesh
   )
   &AddVelocity(
    Cone Angle=10
    Module.UseConeFalloff=false
    Module.UseVelDistribution=false
    Override Seed=false
    Velocity Mode=::NewEnumerator2
    Velocity Speed=-400
   )
   /SetVariables(Particles.SVisTag:NiagaraInt32=&SpawnGroup()) {isSetParameters=true, module=SetVariables_#2}
  ]
  particleUpdateScript [
   &FN_DisableTireFX(
    Mask Index Attribute:NiagaraInt32=$Particles.PartitionParticles.Partition
    Mask Index FL=2
    Mask Index FR=3
    Mask Index RL=0
    Mask Index RR=1
   )
   &Color(Color=&VC_DynamicScaleColor(
     Color to Scale:LinearColor=&MakeCustomLinearColorFromBool(
      Bool=$User.bIsPainted
      True Color=$User.PaintedColor
      False Color=&SelectLinearColorFromArray(
       Array Sampling Mode=::NewEnumerator0
       Direct Array Index=?
       Color Selection Array=DataInterface() {propertyValues {ColorData [<rgba 1 0 0.3152 1>, <rgba 1 1 1 1>, <rgba 0.5674 1 0.0923 1>, <rgba 0 0 0 0>, <rgba 0 0 0 0>, <rgba 0 0 0 0>]}}
       Random Seed=0
      )
     )
     RGB Scale:Vector3f=<1 1 1>
     Alpha Scale:NiagaraFloat=1
    )) {enabled=false}
   &ScaleColor(
    Scale Alpha=&FloatFromCurve(FloatCurve=DataInterface() {propertyValues {Curve {keys [<at 0 2>,<at 1 0>]}}}, Scale Curve=0.5)
    Scale RGB=<2 2 2>
    ScaleA=true
    ScaleRGB=true
   )
   &Drag(
    Drag=&FloatFromCurve(FloatCurve=DataInterface() {propertyValues {Curve {keys [{tangentMode=RCTM_User, time=0, value=0}, {tangentMode=RCTM_User, time=1, value=0.5}]}}})
    Use Linear Drag=true
    Use Rotational Drag=true
   )
   &ParticleState(Kill Particles When Lifetime Has Elapsed=false)
   &ScaleSpriteSize(Uniform Curve Sprite Scale=DataInterface() {propertyValues {Curve {keys [{tangentMode=RCTM_User, time=0, value=0.5}, {tangentMode=RCTM_User, time=1, value=1}]}}})
   /SetVariables() {isSetParameters=true, module=SetVariables_#3}
   /SetVariables(Particles.StarBitColor:LinearColor=&VC_DynamicScaleColor(
     Color to Scale:LinearColor=&MakeCustomLinearColorFromBool(Bool=$User.bIsPainted, True Color=$User.StarCrossColor, False Color=<rgba 5 5 5 1>)
     RGB Scale:Vector3f=<5 5 5>
     Alpha Scale:NiagaraFloat=&FloatFromCurve(FloatCurve=DataInterface() {propertyValues {Curve {keys [<at 0 2>,<at 1 0>]}}})
    )) {isSetParameters=true, module=SetVariables_#4}
   &SolveForcesAndVelocity(Clamp Velocity=false, Limit Acceleration=false)
  ]
  rendererClasses [&NiagaraSpriteRendererProperties]
  renderers [
   {rendererClass=&NiagaraSpriteRendererProperties}
   {rendererClass=&NiagaraSpriteRendererProperties, rendererIndex=1}
  ]
 }
 NE_VC_DriftTrail_Sprite_Parent004 {
  emitterUpdateScript [
   &EmitterState(
    Inactive Response=::NewEnumerator1
    Life Cycle Mode=::NewEnumerator1
    Loop Behavior=::NewEnumerator1
    Loop Duration Mode=::NewEnumerator1
    UseLoopDelay=false
   )
   &SpawnBurst_Instantaneous(Spawn Count=2, Use Spawn Probability=false, UseLoopCountLimit=false)
   &SpawnBurst_Instantaneous(Spawn Count=2, Spawn Group=1, Use Spawn Probability=false, UseLoopCountLimit=false) {module=SpawnBurst_Instantaneous001}
   &SpawnBurst_Instantaneous(Spawn Count=2, Spawn Group=2, Use Spawn Probability=false, UseLoopCountLimit=false) {module=SpawnBurst_Instantaneous002}
   &SpawnRate(Use Spawn Probability=false) {enabled=false}
  ]
  particleSpawnScript [
   &VC_PaintedVariantsOverride() {enabled=false}
   /SetVariables(Particles.DVisTag:NiagaraInt32=&SpawnGroup()) {isSetParameters=true, module=SetVariables_#5}
   &InitializeParticle(
    Color Mode=::NewEnumerator0
    Lifetime=$System.CurrentLoopDuration
    Sprite Rotation Mode=::NewEnumerator2
    Sprite Size Mode=::NewEnumerator3
    Uniform Sprite Size=&VC_FloatBySpawnGroup(
     SpawnGroup or ExecIndex:NiagaraInt32=0
     CustomInt:NiagaraInt32=0
     MaximumSpawnGroups:NiagaraInt32=3
     Float SG 0:NiagaraFloat=70
     Float SG 1:NiagaraFloat=10
     Float SG 2:NiagaraFloat=&MakeCustomFloatFromBool(Bool=$User.bIsPainted, True Float=90, False Float=70)
     Float SG 3:NiagaraFloat=0
     Float SG 4:NiagaraFloat=0
     Float SG 5:NiagaraFloat=0
    )
    UsePositionOffset=false
    Write Material Random=true
   )
   &PartitionParticles()
   &SkeletalMeshLocation(
    Bone / Socket Index=&Add_Integer(A=$Particles.PartitionParticles.Partition, B=4)
    Bone / Socket Sampling Mode=::NewEnumerator5
    Interpolate Mesh Sampling=false
    MaskSpawnGroup=false
    Mesh Sampling Type=::NewEnumerator6
    Sampled Position Offset=&AddVectorToPosition(Position=<0 0 0>, Vector=Linked() {linkedVariable {name=User.offset, type=NiagaraPosition}})
    Skeletal Mesh=$System.SkeletalMesh
   )
   &AddVelocity(Velocity=&VC_Mask_Vectors(
     FrontEnd Value:Vector3f=<-1500 0 0>
     blsFrontEndPreview:NiagaraBool=false
     SpeedMin Value:Vector3f=<-800 0 0>
     SpeedMax Value:Vector3f=<-1 0 0>
     Speed Alpha:NiagaraFloat=$System.Speed_Alpha
    )) {enabled=false}
  ]
  particleUpdateScript [
   &FN_DisableTireFX(
    Mask Index Attribute:NiagaraInt32=$Particles.PartitionParticles.Partition
    Mask Index FL=2
    Mask Index FR=3
    Mask Index RL=0
    Mask Index RR=1
   )
   &ParticleState(Kill Particles When Lifetime Has Elapsed=false, Loop Particles Lifetime=true)
   &Color(Color=&MakeCustomLinearColorFromBool(Bool=$User.bIsPainted, True Color=$User.RaysColor, False Color=<rgba 5 5 5 1>))
   &DynamicMaterialParameters(
    Index 0 Param 1=&MakeCustomFloatFromBool(Bool=$User.bIsPainted, True Float=0, False Float=1)
    Index 0 Param 2=1
    Index 0 Param 3=0.5
    Index 0 Param 4=1
    Index 1 Param 1=&MakeCustomFloatFromBool(Bool=$User.bIsPainted, True Float=0, False Float=1)
    Index 2 Param 1=&MakeCustomFloatFromBool(Bool=$User.bIsPainted, True Float=1, False Float=1)
    Index 2 Param 2=&MakeCustomFloatFromBool(Bool=$User.bIsPainted, True Float=1, False Float=0)
    Module.Index0Float0Write=true
    Module.Index0Float1Write=true
    Module.Index0Float2Write=true
    Module.Index0Float3Write=true
    Module.Index1Float0Write=true
    Module.Index1Float1Write=true
    Module.Index1Float2Write=true
    Module.Index1Float3Write=true
    Module.Index2Float0Write=true
    Module.Index2Float1Write=true
    Module.Index2Float2Write=true
    Module.Index2Float3Write=true
    Write Parameter Index 1=::NewEnumerator1
    Write Parameter Index 2=::NewEnumerator1
   )
   &SkeletalMeshLocation(
    Bone / Socket Index=&Add_Integer(A=$Particles.PartitionParticles.Partition, B=4)
    Bone / Socket Sampling Mode=::NewEnumerator5
    MaskSpawnGroup=false
    Mesh Sampling Type=::NewEnumerator6
    Sampled Position Offset=&AddVectorToPosition(Position=<-65 300 -20>, Vector=Linked() {linkedVariable {name=User.offset, type=NiagaraPosition}})
    Skeletal Mesh=$System.SkeletalMesh
   ) {module=SkeletalMeshLocation001}
   /SetVariables(
    Particles.GlowColor=&VC_DynamicScaleColor(
     Color to Scale:LinearColor=&MakeCustomLinearColorFromBool(Bool=$User.bIsPainted, True Color=$User.GlowColor, False Color=<rgba 1 1 1 1>)
     RGB Scale:Vector3f=<1 1 1>
     Alpha Scale:NiagaraFloat=0.5
    )
    Particles.SoftDotColor:LinearColor=&VC_DynamicScaleColor(
     Color to Scale:LinearColor=&MakeCustomLinearColorFromBool(
      Bool=$User.bIsPainted
      True Color=&Multiply_LinearColorByFloat(Linear Color=$User.SoftDotColor, Float=2)
      False Color=<rgba 4 4 4 1>
     )
     RGB Scale:Vector3f=<1 1 1>
     Alpha Scale:NiagaraFloat=1
    )
   ) {isSetParameters=true}
   &SolveForcesAndVelocity(Clamp Velocity=false, Limit Acceleration=false)
  ]
  rendererClasses [&NiagaraSpriteRendererProperties]
  renderers [
   {rendererClass=&NiagaraSpriteRendererProperties}
   {rendererClass=&NiagaraSpriteRendererProperties, rendererIndex=1}
   {rendererClass=&NiagaraSpriteRendererProperties, rendererIndex=2}
  ]
 }
 NE_VC_DriftTrail_WheelWrap_Parent {
  emitterUpdateScript [
   &EmitterState(
    Inactive Response=::NewEnumerator1
    Life Cycle Mode=::NewEnumerator1
    Loop Behavior=::NewEnumerator1
    Loop Duration Mode=::NewEnumerator1
    UseLoopDelay=false
   )
   &SpawnRate(SpawnRate=10, Use Spawn Probability=false) {enabled=false}
   &SpawnBurst_Instantaneous(Spawn Count=2, Use Spawn Probability=false, UseLoopCountLimit=false)
  ]
  particleSpawnScript [
   &VC_PaintedVariantsOverride(
    Black=<rgba 0.005 0.005 0.005 5>
    BurntSienna=<rgba 0.02 0.0062 0.0025 5>
    Cobalt=<rgba 0 0.0017 0.09 5>
    Crimson=<rgba 0.15 0 0 2>
    DefaultColor=<rgba 0.5 0.5 0.5 1>
    ForestGreen=<rgba 0 0.25 0 1>
    Grey=<rgba 0.02 0.02 0.02 5>
    Lime=<rgba 0.075 0.15 0 3>
    Orange=<rgba 0.5 0.0833 0 1>
    Pink=<rgba 0.3 0.003 0.1515 1>
    Purple=<rgba 0.0063 0.0005 0.05 4>
    Saffron=<rgba 0.4 0.2 0 1>
    SkyBlue=<rgba 0 0.1996 0.3 2>
    TitaniumWhite=<rgba 0.15 0.15 0.15 2>
   )
   &InitializeParticle(
    Color=$Particles.Color
    Mesh Scale Mode=::NewEnumerator4
    Sprite Rotation Mode=::NewEnumerator2
    Sprite Size Mode=::NewEnumerator3
    UsePositionOffset=false
    Write Material Random=true
   )
   &PartitionParticles()
   &SkeletalMeshLocation(
    Bone / Socket Index=&Add_Integer(A=$Particles.PartitionParticles.Partition, B=4)
    Bone / Socket Sampling Mode=::NewEnumerator5
    Interpolate Mesh Sampling=false
    MaskSpawnGroup=false
    Mesh Sampling Type=::NewEnumerator6
    Skeletal Mesh=$System.SkeletalMesh
   )
  ]
  particleUpdateScript [
   /SetVariables(Particles.Linear Color=&VC_DynamicScaleColor(
     Color to Scale:LinearColor=$Particles.Color
     RGB Scale:Vector3f=&VectorFromFloat(Value=10)
     Alpha Scale:NiagaraFloat=1
    )) {isSetParameters=true, module=SetVariables_#6}
   &FN_DisableTireFX(
    Mask Index Attribute:NiagaraInt32=$Particles.PartitionParticles.Partition
    Mask Index FL=2
    Mask Index FR=3
    Mask Index RL=0
    Mask Index RR=1
   )
   &VC_WheelWrapScale(Wheel Index Attribute:NiagaraInt32=&Add_Integer(A=$Particles.PartitionParticles.Partition, B=2))
   &ParticleState(Kill Particles When Lifetime Has Elapsed=false)
   &ScaleColor(
    Scale Alpha=&MakeCustomFloatFromBool(Bool=$User.bIsPainted, True Float=1.5, False Float=1)
    Scale RGB=&MakeCustomVectorFromBool(
     Bool=$User.bIsPainted
     True Vector=&MakeVectorFromLinearColorRGB(Color=$User.PaintedColor)
     False Vector=<1 1 1>
    )
    ScaleA=true
    ScaleRGB=true
   ) {enabled=false}
   &UpdateMeshOrientation(
    Rotation Coordinate Space=::NewEnumerator0
    Rotation Rate=1000
    Rotation Vector=&SelectVectorFromArray(
     Direct Array Index=$Particles.Initial.PartitionParticles.Partition
     Vector Selection Array=DataInterface() {propertyValues {FloatData [<1 -1 1>, <1 1 1>]}}
    )
   ) {enabled=false}
   &DynamicMaterialParameters(
    Index 0 Param 1=&MakeCustomFloatFromBool(Bool=$User.bIsPainted, True Float=0, False Float=1)
    Module.Index0Float0Write=true
    Module.Index0Float1Write=true
    Module.Index0Float2Write=true
    Module.Index0Float3Write=true
   )
   &SkeletalMeshLocation(
    Bone / Socket Index=&Add_Integer(A=$Particles.PartitionParticles.Partition, B=4)
    Bone / Socket Sampling Mode=::NewEnumerator5
    MaskSpawnGroup=false
    Mesh Sampling Type=::NewEnumerator6
    Skeletal Mesh=$System.SkeletalMesh
   ) {module=SkeletalMeshLocation001}
   &SolveForcesAndVelocity(Clamp Velocity=false, Limit Acceleration=false)
  ]
  rendererClasses [&NiagaraMeshRendererProperties]
  renderers [{rendererClass=&NiagaraMeshRendererProperties, rendererIndex=0}]
 }
 NE_VC_DriftTrail_WheelWrap_Parent001 {
  emitterUpdateScript [
   &EmitterState(
    Inactive Response=::NewEnumerator1
    Life Cycle Mode=::NewEnumerator1
    Loop Behavior=::NewEnumerator1
    Loop Duration Mode=::NewEnumerator1
    UseLoopDelay=false
   )
   &SpawnRate(SpawnRate=10, Use Spawn Probability=false) {enabled=false}
   &SpawnBurst_Instantaneous(Spawn Count=2, Use Spawn Probability=false, UseLoopCountLimit=false)
  ]
  particleSpawnScript [
   &VC_PaintedVariantsOverride(
    Black=<rgba 0.005 0.005 0.005 5>
    BurntSienna=<rgba 0.02 0.0062 0.0025 5>
    Cobalt=<rgba 0 0.0017 0.09 5>
    Crimson=<rgba 0.15 0 0 2>
    DefaultColor=<rgba 0.5 0.5 0.5 1>
    ForestGreen=<rgba 0 0.25 0 1>
    Grey=<rgba 0.02 0.02 0.02 5>
    Lime=<rgba 0.075 0.15 0 3>
    Orange=<rgba 0.5 0.0833 0 1>
    Pink=<rgba 0.3 0.003 0.1515 1>
    Purple=<rgba 0.0063 0.0005 0.05 4>
    Saffron=<rgba 0.4 0.2 0 1>
    SkyBlue=<rgba 0 0.1996 0.3 2>
    TitaniumWhite=<rgba 0.15 0.15 0.15 2>
   )
   &InitializeParticle(
    Color=$Particles.Color
    Mesh Scale Mode=::NewEnumerator4
    Sprite Rotation Mode=::NewEnumerator2
    Sprite Size Mode=::NewEnumerator3
    UsePositionOffset=false
    Write Material Random=true
   )
   &PartitionParticles()
   &SkeletalMeshLocation(
    Bone / Socket Index=&Add_Integer(A=$Particles.PartitionParticles.Partition, B=4)
    Bone / Socket Sampling Mode=::NewEnumerator5
    Interpolate Mesh Sampling=false
    MaskSpawnGroup=false
    Mesh Sampling Type=::NewEnumerator6
    Skeletal Mesh=$System.SkeletalMesh
   )
  ]
  particleUpdateScript [
   /SetVariables(Particles.Linear Color=&VC_DynamicScaleColor(
     Color to Scale:LinearColor=$Particles.Color
     RGB Scale:Vector3f=&VectorFromFloat(Value=10)
     Alpha Scale:NiagaraFloat=1
    )) {isSetParameters=true, module=SetVariables_#6}
   &FN_DisableTireFX(
    Mask Index Attribute:NiagaraInt32=$Particles.PartitionParticles.Partition
    Mask Index FL=2
    Mask Index FR=3
    Mask Index RL=0
    Mask Index RR=1
   )
   &VC_WheelWrapScale(Wheel Index Attribute:NiagaraInt32=&Add_Integer(A=$Particles.PartitionParticles.Partition, B=2))
   &ParticleState(Kill Particles When Lifetime Has Elapsed=false)
   &ScaleColor(
    Scale Alpha=&MakeCustomFloatFromBool(Bool=$User.bIsPainted, True Float=1.5, False Float=1)
    Scale RGB=&MakeCustomVectorFromBool(
     Bool=$User.bIsPainted
     True Vector=&MakeVectorFromLinearColorRGB(Color=$User.PaintedColor)
     False Vector=<1 1 1>
    )
    ScaleA=true
    ScaleRGB=true
   ) {enabled=false}
   &UpdateMeshOrientation(
    Rotation Coordinate Space=::NewEnumerator0
    Rotation Rate=1000
    Rotation Vector=&SelectVectorFromArray(
     Direct Array Index=$Particles.Initial.PartitionParticles.Partition
     Vector Selection Array=DataInterface() {propertyValues {FloatData [<1 -1 1>, <1 1 1>]}}
    )
   ) {enabled=false}
   &DynamicMaterialParameters(
    Index 0 Param 1=&MakeCustomFloatFromBool(Bool=$User.bIsPainted, True Float=0, False Float=1)
    Module.Index0Float0Write=true
    Module.Index0Float1Write=true
    Module.Index0Float2Write=true
    Module.Index0Float3Write=true
   )
   &SkeletalMeshLocation(
    Bone / Socket Index=&Add_Integer(A=$Particles.PartitionParticles.Partition, B=4)
    Bone / Socket Sampling Mode=::NewEnumerator5
    MaskSpawnGroup=false
    Mesh Sampling Type=::NewEnumerator6
    Sampled Position Offset=<0 300 0>
    Skeletal Mesh=$System.SkeletalMesh
   ) {module=SkeletalMeshLocation001}
   &SolveForcesAndVelocity(Clamp Velocity=false, Limit Acceleration=false)
  ]
  rendererClasses [&NiagaraMeshRendererProperties]
  renderers [{rendererClass=&NiagaraMeshRendererProperties, rendererIndex=0}]
 }
}
summary {
 systemName=NS_DriftTrail_Disco_Backup
 emitters [
  {emitterName=NE_Disco_Ribbon, rendererClasses [&NiagaraRibbonRendererProperties]}
  {emitterName=NE_VC_DriftTrail_Sprite_Parent, rendererClasses [&NiagaraSpriteRendererProperties]}
  {emitterName=NE_VC_DriftTrail_Sprite_Parent001, rendererClasses [&NiagaraSpriteRendererProperties]}
  {emitterName=NE_VC_DriftTrail_WheelWrap_Parent, rendererClasses [&NiagaraMeshRendererProperties]}
  {emitterName=NE_Disco_Ribbon001, rendererClasses [&NiagaraRibbonRendererProperties]}
  {emitterName=NE_VC_DriftTrail_Sprite_Parent003, rendererClasses [&NiagaraSpriteRendererProperties]}
  {emitterName=NE_VC_DriftTrail_Sprite_Parent004, rendererClasses [&NiagaraSpriteRendererProperties]}
  {emitterName=NE_VC_DriftTrail_WheelWrap_Parent001, rendererClasses [&NiagaraMeshRendererProperties]}
 ]
 userVariables [
  User.bIsFrontendPreview {type=NiagaraBool}
  User.bIsPainted {type=NiagaraBool}
  User.ColorVariantID {type=NiagaraInt32}
  User.GlowColor {type=LinearColor, description="", defaultValue {struct=LinearColor, value=<rgba 1 1 1 1>}}
  User.LParticleColor {type=LinearColor, description="", defaultValue {struct=LinearColor, value=<rgba 1 1 1 1>}}
  User.offset {type=NiagaraPosition, description="", defaultValue {}}
  User.PaintedColor {type=LinearColor, description="", defaultValue {struct=LinearColor, value=<rgba 1 1 1 1>}}
  User.PaintedScalar {type=NiagaraFloat, defaultValue=1, description=""}
  User.RaysColor {type=LinearColor, description="", defaultValue {struct=LinearColor, value=<rgba 1 1 1 1>}}
  User.SoftDotColor {type=LinearColor, description="", defaultValue {struct=LinearColor, value=<rgba 1 1 1 1>}}
  User.StarCrossColor {type=LinearColor, description="", defaultValue {struct=LinearColor, value=<rgba 1 1 1 1>}}
  User.VehicleFlags {type=NiagaraInt32, defaultValue=806289408}
  User.Wheel_Radius_Array {
   type=NiagaraDataInterfaceArrayFloat
   defaultValue {
    struct=NiagaraExt_VariableValue_DataInterface
    value {
     dataInterface=&self:NiagaraDataInterfaceArrayFloat_1
     dataInterfaceClass=&NiagaraDataInterfaceArrayFloat
    }
   }
  }
  User.Wheel_Width_Array {
   type=NiagaraDataInterfaceArrayFloat
   defaultValue {
    struct=NiagaraExt_VariableValue_DataInterface
    value {
     dataInterface=&self:NiagaraDataInterfaceArrayFloat_0
     dataInterfaceClass=&NiagaraDataInterfaceArrayFloat
    }
   }
  }
  User.WheelTouchingArray {
   type=NiagaraDataInterfaceArrayFloat
   defaultValue {
    struct=NiagaraExt_VariableValue_DataInterface
    value {
     dataInterface=&self:NiagaraDataInterfaceArrayFloat_2
     dataInterfaceClass=&NiagaraDataInterfaceArrayFloat
    }
   }
  }
  User.WheelWColor {type=LinearColor, description="", defaultValue {struct=LinearColor, value=<rgba 1 1 1 1>}}
 ]
}
